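; Drum kit stage actor for Sparks music play that dresses itself in the instigating player's drum cosmetics.
; FixupDrumCosmetics applies mesh and materials and respawns attached VFX, though it queries the "Guitar" slot for VFX.
; On cosmetic-capable clients, BeginPlay also tints dynamic materials by weapon rarity, enabling the power-up glow above Common.
; It follows the music playspace's MIDI driver, clearing that glow on "music_end" for kits left on stage.
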(event EventBeginPlay
  (bind _skeletalmesh (Variables|Default|GetSkeletalMesh))
  ;[1 1 1] See: PilgrimQuickplayStateMachine::PrepareStageForPlayer
  ; for actual spawning of this kit
  (Networking|CanExecuteCosmeticEvents
    (:True
      (Utilities|FlowControl|Sequence
        (:then_0
          (bind _asfort_pawn (Utilities|Casting|CastToFortPawn (Variables|Actor|GetInstigator)))
          (Utilities|IsValid (Weapon|GetCurrentWeapon _asfort_pawn)
            (:Is%20Valid
              (CallFunction|FixupDrumCosmetics _asfort_pawn)
              (for (_ _array_index) (Rendering|Material|GetMaterials _skeletalmesh)
                (bind _returnvalue (Fortnite|GetDynamicMaterialInstance _skeletalmesh _array_index))
                (Utilities|IsValid _returnvalue
                  (:Is%20Valid
                    (bind _returnvalue_2 (Fortnite|GetRarityColorPalette (Rarity|GetWeaponDataRarity (Weapon|GetCurrentWeapon _asfort_pawn))))
                    (Rendering|Material|SetVectorParameterValue _returnvalue "Power Up Rarity Color" (Utilities|Struct|BreakFortColorPalette _returnvalue_2) :!class /Script/Engine.MaterialInstanceDynamic)
                    (Rendering|Material|SetScalarParameterValue _returnvalue "Power Up Generic On/Off" (Math|Float|SelectFloat 1.0 0.0 (Utilities|Enum|NotEqual(Enum) (Rarity|GetWeaponDataRarity (Weapon|GetCurrentWeapon _asfort_pawn)) "Common")) :!class /Script/Engine.MaterialInstanceDynamic)))))))
        (:then_1
          ;[1 0.4757 0.8337] Listen for song changes
          (bind (_asynctaskproxy _playspace) (Playspace|Discoverable|WaitForDiscoverablePlayspaceOfTypeForActor "/Script/SparksMusicPlayspaceRuntime.SparksMusicPlayspace" (Variables|Actor|GetInstigator))
            (:OnFoundPlayspace
              (bind _assparks_music_playspace (Utilities|Casting|CastToSparksMusicPlayspace _playspace)
                (:then
                  (bind _output_get (Variables|Default|SetSparksMusicPlayspace _assparks_music_playspace))
                  (Utilities|IsValid (Variables|/Script/SparksMusicPlayspaceRuntime.SparksMusicPlayspace|GetMidiEventDataManager _output_get)
                    (:Is%20Valid
                      (EventDispatchers|BindEventtoOnMidiEventDriverChanged (Variables|/Script/SparksMusicPlayspaceRuntime.SparksMusicPlayspace|GetMidiEventDataManager _output_get) (EventDispatchers|CreateEvent :!event Custom|OnMidiEventDriverChanged_Event))
                      (CallFunction|OnMidiEventDriverChangedEvent (Class|ActorComponentMidiEventDataManager|TryGetMidiEventDriverforPlayer (Variables|/Script/SparksMusicPlayspaceRuntime.SparksMusicPlayspace|GetMidiEventDataManager _output_get) (Variables|Actor|GetInstigator))))))
                (:CastFailed
                  (Fortnite|FortShippingLog self (Utilities|String|Append "WaitForPlayspace: Given playspace " (Utilities|GetDisplayName _playspace) " which doesn't derive from SparksMusicPlayspace.") true))))))))))

(event Custom|OnMidiEventDriverChanged_Event
  (params (MidiEventDriver :type Midi%20Event%20Driver%20Interface%20Interface))
  (bind _skeletalmesh (Variables|Default|GetSkeletalMesh))
  ;[1 0.6625 0.4583] Listen for the end of the song
  (if (Utilities|IsValidInterface MidiEventDriver)
    (bind _output_get (Variables|Default|SetMidiEventRouter MidiEventDriver))
    (MidiEvents|ListenForMidiTextEvent _output_get "EVENTS" "music_end"
      (:OnMidiEventStart
        ;[0 1 0.4643] Turn off any powerup effects in the materials if this is a drum kit we are leaving on stage
        (for (_ _array_index) (Rendering|Material|GetMaterials _skeletalmesh)
          (bind _returnvalue (Fortnite|GetDynamicMaterialInstance _skeletalmesh _array_index))
          (Rendering|Material|SetScalarParameterValue _returnvalue "Power Up Generic On/Off" 0.0 :!class /Script/Engine.MaterialInstanceDynamic))))))

(event EventOnWorldReady
  (Networking|CanExecuteCosmeticEvents
    (:True
      (CallFunction|FixupDrumCosmetics 0))))

(fn DestroyCosmeticVFX
  (bind _activecosmeticvfx (Variables|Default|GetActiveCosmeticVFX))
  (Utilities|IsValid _activecosmeticvfx
    (:Is%20Valid
      (Components|DestroyComponent _activecosmeticvfx)
      (Variables|Default|SetActiveCosmeticVFX 0))))

(fn FixupDrumCosmetics
  (params (Player :type Fort%20Pawn%20Object%20Reference))
  (bind _returnvalue (Actor|GetComponentbyClass Player "/Script/SparksCoreCosmeticsRuntime.SparksCosmeticComponent"))
  (bind (_outfx _outauxfxsocket _returnvalue_2) (Class|SparksCosmeticComponent|GetInstrumentAuxiliaryVFX _returnvalue "Guitar"))
  (bind (_outmesh _outmaterial _outmaterial2 _returnvalue_3) (Class|SparksCosmeticComponent|GetInstrumentAuxiliaryCosmetics _returnvalue "Drum"))
  (bind _skeletalmesh (Variables|Default|GetSkeletalMesh))
  (Utilities|IsValid Player
    (:Is%20Valid
      (Utilities|FlowControl|Sequence
        (:then_0
          ;[1 1 1] set up mesh and materials
          (Utilities|IsValid _returnvalue
            (:Is%20Valid
              (if _returnvalue_3
                (Components|SkeletalMesh|SetSkeletalMeshAsset _skeletalmesh _outmesh)
                (Rendering|Material|SetMaterial _skeletalmesh 0 _outmaterial)
                (Rendering|Material|SetMaterial _skeletalmesh 1 _outmaterial2)))))
        (:then_1
          ;[1 1 1] Attach VFX
          (CallFunction|DestroyCosmeticVFX)
          (if _returnvalue_2
            (bind _returnvalue_4 (Effects|SpawnFXSystemAttached _outfx (Variables|Default|GetSkeletalMesh)))
            (Variables|Default|SetActiveCosmeticVFX _returnvalue_4)))))
    (:Is%20Not%20Valid
      (bind _asfort_pawn (Utilities|Casting|CastToFortPawn (Actor|GetOwner)))
      (CallFunction|FixupDrumCosmetics _asfort_pawn))))

(fn GetEquippedQuickbarItem
  (params (QuickbarType :type EFortQuickBars%20Enum :default "Primary") (FortPlayerController :type Fort%20Player%20Controller%20Object%20Reference))
  ;[1 0.07087 0.3752] Client Only! Loop through the quick bar for the associated player controller and return the equipped item
  (bind _returnvalue (Inventory|GetNumQuickbarSlots FortPlayerController QuickbarType))
  (bind _index (Utilities|FlowControl|ForLoopwithBreak 0 _returnvalue)
    (:LoopBody
      (bind _returnvalue_2 (Inventory|GetIteminQuickbarSlot FortPlayerController QuickbarType _index))
      (Utilities|IsValid _returnvalue_2
        (:Is%20Valid
          (if (Item|IsEquipped _returnvalue_2)
            (return _returnvalue_2)))))
    (:Completed
      ;[0.8449 1 0] No equipped item was found
      (return 0))))

(fn ConstructionScript)
